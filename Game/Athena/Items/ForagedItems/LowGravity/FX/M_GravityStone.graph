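# Opaque material for the low-gravity foraged Gravity Stone: a glossy, metallic body with a floating, spinning animation.
# A Fresnel-driven thin-film spectrum texture tints the base color with iridescence.
# Emissive combines panning bump-offset smoke and glow color, masked by inverse Fresnel and a pulsing, squashed sphere mask, then boosted to HDR, with a near-zero low-quality fallback.
# Switchable WPO adds Random-varied Z rotation and sine bobbing, with rotated normals correcting shading.
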Material /Game/Athena/Items/ForagedItems/LowGravity/FX/M_GravityStone
graph {
 properties {
  bTangentSpaceNormal=false
  bUseMaterialAttributes=true
  bUsedWithMeshParticles=true
  bUsedWithNanite=true
  pixelDepthOffsetMode=PDOM_Legacy
  usedShadingModels=Default Lit
 }
 nodes [
  n1 = VectorParameter() {parameterName=Color, defaultValue=<rgba 0.0845 0.0877 0.1 1>}
  n2 = ScalarParameter() {defaultValue=1, parameterName=Random}
  n3 = Time()
  n4 = Multiply(A=@n3, B=Multiply(A=@n2) {constB=0.2}) {constB=0.1}
  n5 = MaterialFunctionCall(Rotation Axis=Constant3Vector() {constant=<rgba 0 0 1 0>}, Rotation Angle=@n4) {materialFunction=&FixRotateAboutAxisNormals}
  n6 = StaticSwitchParameter(True=@n5.New World Space Vertex Normal, False=PixelNormalWS()) {defaultValue=true, parameterName=UseWPO}
  n7 = Multiply(
   A=TextureSampleParameter2D(UVs=Fresnel(Normal=@n6) {baseReflectFraction=0, exponent=0.5}) {group=ThinFilm, parameterName=Thin Film Texture, texture=/Game/Global/Textures/Default/Thinfilm/T_ThinFilm_Spectrum_COLOR}.RGB
   B=ScalarParameter() {defaultValue=0.8, group=ThinFilm, parameterName=ThinFilm_Intensity}
  )
  n8 = Fresnel(ExponentIn=ScalarParameter() {defaultValue=0.9, group=ThinFilm, parameterName=ThinFilm_Visibility}, Normal=@n6) {baseReflectFraction=0}
  n9 = Constant() {r=0.8}
  n10 = TextureCoordinate() {uTiling=1.5, vTiling=1.5}
  n11 = ScalarParameter() {defaultValue=-4, parameterName=BumpOffset}
  n12 = TextureSample(UVs=Panner(Coordinate=BumpOffset(Coordinate=@n10, Height=@n11)) {speedX=-0.25, speedY=-0.16}) {texture=&T_Inky_Smoke_Tile_Chromatic}
  n13 = Clamp(None=Power(Base=OneMinus(None=@n8), Exp=ScalarParameter() {defaultValue=4.5, parameterName=GlowAngleFade}) {constExponent=4})
  n14 = Multiply(
   A=Subtract(A=WorldPosition().XYZ, B=ActorPositionWS())
   B=VectorParameter() {parameterName=SpheremaskScaleMult, defaultValue=<rgba 1 1 0.3 1>}.RGB
  )
  n15 = VectorParameter() {parameterName=SphereMaskOffset, defaultValue=<rgba 0 0 10 1>}
  n16 = Sine(None=@n3) {period=1.14}
  n17 = Multiply(
   A=LinearInterpolate(Alpha=ConstantBiasScale(None=@n16)) {constA=0.95, constB=1.05}
   B=ScalarParameter() {defaultValue=11, parameterName=GlowWorldRadius}
  )
  n18 = Multiply(A=Power(Base=SphereMask(A=@n14, B=@n15.RGB, Radius=@n17) {attenuationRadius=11, hardnessPercent=0.1}) {constExponent=6}) {constB=2}
  n19 = Multiply(
   A=Add(A=Multiply(A=@n12.RGB) {constB=4}, B=VectorParameter() {parameterName=GlowColor, defaultValue=<rgba 0.6783 1.4419 5 0>}.RGB)
   B=QualitySwitch(Default=Multiply(A=@n13, B=@n18), Low=Constant() {r=0.005})
  )
  n20 = Multiply(A=Sine(None=Add(A=@n3, B=@n2)) {period=3.3}, B=Constant3Vector() {constant=<rgba 0 0 3 0>})
  n21 = MaterialFunctionCall!(Rotation Amount=@n4, PivotPoint, WorldPosition) {materialFunction=/Engine/Functions/Engine_MaterialFunctions02/WorldPositionOffset/RotateAboutWorldAxis_cheap}
  MakeMaterialAttributes!(
   BaseColor=LinearInterpolate(A=@n1.RGB, B=Add(A=@n7, B=@n1.RGB), Alpha=@n8)
   Metallic=@n9
   Specular=@n9
   Roughness=Constant() {r=0.02}
   Anisotropy
   EmissiveColor=MaterialFunctionCall(LDR Input=Clamp(None=@n19), HDR Tint=@n19, HDR Intensity Multiplier=ScalarParameter() {defaultValue=15, parameterName=HDRIntensityMult}) {materialFunction=&DeriveHDRfromLDR}.RemappedRange
   Opacity
   OpacityMask
   Normal=StaticSwitchParameter(True=@n5.New World Space Vertex Normal, False=VertexNormalWS()) {defaultValue=true, parameterName=UseWPO}
   Tangent
   WorldPositionOffset=StaticSwitchParameter(True=Add(A=@n20, B=@n21.Z-Axis), False=Constant()) {defaultValue=true, parameterName=UseWPO}
   SubsurfaceColor
   ClearCoat
   ClearCoatRoughness
   AmbientOcclusion
   Refraction
   CustomizedUVs_0
   CustomizedUVs_1
   CustomizedUVs_2
   CustomizedUVs_3
   CustomizedUVs_4
   CustomizedUVs_5
   CustomizedUVs_6
   CustomizedUVs_7
   PixelDepthOffset
   ShadingModel
   Displacement
  )
 ]
}
