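# Niagara system for the B variant of the pickup effect on the Low Gravity foraged item.
# It comprises three emitters, Trail, Flare and Embers, with no renderers listed and no user parameters exposed.
# At system level, SystemState runs alongside two Set Variables modules, with float-multiply and float-from-vector dynamic inputs in use and no data interfaces.
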NiagaraSystem /Game/Athena/Items/ForagedItems/LowGravity/FX/NS_LowGravity_Pickup_B
dependencies {
 usedDataInterfaces []
 usedDynamicInputs [&Multiply_Float, &MakeFloatFromVector]
 usedModules [
  &self:SystemSpawnScript.SystemScriptSource.SystemScriptGraph.NiagaraNodeAssignment_0.SetVariables_#1
  &self:SystemSpawnScript.SystemScriptSource.SystemScriptGraph.NiagaraNodeAssignment_1.SetVariables_#2
  &SystemState
 ]
}
emitters {Embers {rendererClasses [], renderers []}, Flare {rendererClasses [], renderers []}, Trail {rendererClasses [], renderers []}}
summary {
 systemName=NS_LowGravity_Pickup_B
 emitters [{emitterName=Trail, rendererClasses []}, {emitterName=Flare, rendererClasses []}, {emitterName=Embers, rendererClasses []}]
 userVariables []
}
