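; Passive gameplay ability for the low-gravity glitched foraged item that credits a Season 10 Mystery quest for consuming different glitched foraged items.
; On activation it checks the avatar for the ConsumeGlitched quest tag, then whether the player is party-assisting or has the quest and objective incomplete.
; If so, it sends a LowGrav-tagged custom stat event to the quest manager.
; It ends immediately on every path and applies no gravity effect itself.
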(event Ability|EventFailedToActivatePassiveAbility
  (params (ActorInfo :type Gameplay%20Ability%20Actor%20Info%20Structure)))

(event Ability|EventActivateAbility
  (bind _returnvalue (Ability|GetAvatarActorfromActorInfo))
  (bind _returnvalue_2 (Fortnite|GetFortPlayerControllerfromActor _returnvalue))
  ;[1 1 1] Glitched foraged items quest
  (Utilities|IsValid _returnvalue
    (:Is%20Valid
      (bind _bdidactorhavetags (Class|AthenaQuestFunctionLibrary|AthenaCheckActorHasTags _returnvalue "(GameplayTags=((TagName=\"Athena.Quests.GlitchedForaged.ConsumeGlitched\")))" true))
      (if _bdidactorhavetags
        (bind (_questisvalid _questcompleted _objectivecompleted) (Default|AthenaCheckQuestandObjectiveCompletion _returnvalue_2 0 "Quest_BR_S10_Mystery_Consume_GlitchedForaged_Different_03"))
        (bind (_bisassistingobjective _bisvalidplayercontroller) (Class|AthenaQuestFunctionLibrary|AthenaPartyAssistCheck _returnvalue_2 "Quest_BR_S10_Mystery_Consume_GlitchedForaged_Different_03"))
        (if (or (and _bisassistingobjective _bisvalidplayercontroller) (Math|Boolean|ANDBoolean _questisvalid (not _questcompleted) (not _objectivecompleted)))
          (Quest|SendComplexCustomStatEvent (Quests|GetQuestManager _returnvalue_2 "Invalid") 0 "(GameplayTags=)" "(GameplayTags=((TagName=\"Athena.Quests.GlitchedForaged.LowGrav\")))" 1)
          (Ability|EndAbility)
          (else
            (Ability|EndAbility)))
        (else
          (Ability|EndAbility))))
    (:Is%20Not%20Valid
      (Ability|EndAbility))))
